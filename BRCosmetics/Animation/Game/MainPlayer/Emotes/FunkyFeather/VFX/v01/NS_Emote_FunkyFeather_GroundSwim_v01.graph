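# Niagara effect for the FunkyFeather GroundSwim emote, circling meshes around the character with twinkling sparkles.
# Three mesh emitters burst looping-lifetime particles, spacing them by normalized index around tilted rings that Emitter.Age scrolls through U Position.
# The two larger groups lerp up from below at start, while a two-particle group uses array-picked offsets and random spin force.
# A sprite emitter starts spawning after a stepped age-curve delay, with waveform-modulated alpha for flicker.
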NiagaraSystem /BRCosmetics/Animation/Game/MainPlayer/Emotes/FunkyFeather/VFX/v01/NS_Emote_FunkyFeather_GroundSwim_v01
dependencies {
 usedDataInterfaces [
  &NiagaraDataInterfaceMeshRendererInfo
  &NiagaraDataInterfaceArrayFloat3
  &NiagaraDataInterfaceCurve
 ]
 usedDynamicInputs []
 usedModules [&SystemState]
}
emitters {
 InfiniteParticleLifetime {
  emitterUpdateScript [
   &EmitterState(
    Inactive Response=::NewEnumerator1
    Life Cycle Mode=::NewEnumerator1
    Loop Behavior=::NewEnumerator1
    Loop Duration Mode=::NewEnumerator1
    UseLoopDelay=false
   )
   &SpawnBurst_Instantaneous(Spawn Count=10, Use Spawn Probability=false, UseLoopCountLimit=false)
  ]
  particleSpawnScript [
   &InitializeParticle(
    Color Mode=::NewEnumerator0
    Mesh Scale Mode=::NewEnumerator3
    Mesh Uniform Scale=&RandomRangeFloat(Minimum=1.75, Maximum=2)
    Position Mode=::NewEnumerator0
    Write Material Random=true
   )
   &InitialMeshOrientation(
    Mesh Orientation Mode=::NewEnumerator1
    Rotation=&RandomRangeVector(Minimum=<0 0 0>, Maximum=<0 0 1>)
    Rotation Coordinate Space=::NewEnumerator2
    Use Rotation Vector=true
   )
  ]
  particleUpdateScript [
   &ParticleState(Kill Particles When Lifetime Has Elapsed=false, Loop Particles Lifetime=true)
   &ShapeLocation(
    MaskSpawnGroup=false
    Non Uniform Scale=&RandomRangeVector(Minimum=<3 0 1>, Maximum=<3 1 1>)
    Offset=&AddVectorToPosition(
     Position=&Lerp_Vector(
      A=<0 0 -110>
      B=<0 0 0>
      Alpha=&ClampFloat(Float=&Multiply_Float(A=$Emitter.Age, B=1.25))
     )
     Vector=&RandomRangeVector(Minimum=<0 -10 0>, Maximum=<0 10 0>)
    )
    Offset Coordinate Space=::Local
    Offset Mode=::NewEnumerator0
    Override Seed=false
    Ring / Disc Distribution Mode:ENiagaraSphereDistributionMode=::NewEnumerator1
    Ring Radius=35
    Rotation Angle=-10
    Rotation Angle Type:ENiagara_AngleInput=::NewEnumerator0
    Rotation Axis=<0 1 0>
    Rotation Coordinate Space=::Simulation
    Rotation Mode=::NewEnumerator1
    Shape Primitive=::NewEnumerator4
    U Position=&Add_Float(
     A=&Multiply_Float(A=$Emitter.Age, B=-0.3)
     B=&ReturnNormalizedExecIndex(Normalized Index Includes One=false)
    )
    UseOwnerScale=true
   ) {module=ShapeLocation001}
  ]
  rendererClasses [&NiagaraMeshRendererProperties]
  renderers [
   {rendererClass=&NiagaraMeshRendererProperties, rendererIndex=0}
   {rendererClass=&NiagaraMeshRendererProperties, rendererIndex=1}
  ]
 }
 InfiniteParticleLifetime001 {
  emitterUpdateScript [
   &EmitterState(
    Inactive Response=::NewEnumerator1
    Life Cycle Mode=::NewEnumerator1
    Loop Behavior=::NewEnumerator1
    Loop Duration Mode=::NewEnumerator1
    UseLoopDelay=false
   )
   &SpawnBurst_Instantaneous(Spawn Count=2, Spawn Probability=0.5, Spawn Time=1.5, Use Spawn Probability=false, UseLoopCountLimit=false)
  ]
  particleSpawnScript [
   &InitializeParticle(
    Color Mode=::NewEnumerator0
    Mesh Scale Mode=::NewEnumerator3
    Mesh Uniform Scale=0.025
    Position Mode=::NewEnumerator0
    Write Material Random=true
   )
   &InitialMeshOrientation(
    Mesh Orientation Mode=::NewEnumerator1
    Rotation=&RandomRangeVector(Minimum=<0 0 0>, Maximum=<0 0 1>)
    Rotation Coordinate Space=::NewEnumerator2
    Use Rotation Vector=true
   )
   &MeshRotationForce(Rotation=&RandomRangeVector(Minimum=<0 0 -2.5>, Maximum=<0 0 2.5>))
   &ApplyInitialForces()
  ]
  particleUpdateScript [
   &ParticleState(Kill Particles When Lifetime Has Elapsed=false, Loop Particles Lifetime=true)
   &SolveForcesAndVelocity(Clamp Velocity=false, Limit Acceleration=false)
   &ShapeLocation(
    MaskSpawnGroup=false
    Non Uniform Scale=<1 0.5 1>
    Offset=&AddVectorToPosition(
     Position=&SelectVectorFromArray(
      Array Sampling Mode=::NewEnumerator0
      Direct Array Index=?
      Vector Selection Array=DataInterface() {propertyValues {FloatData [<0 -33 18>, <1 30 18>]}}
      Random Seed=0
     )
     Vector=&RandomRangeVector(Minimum=<0 -5 0>, Maximum=<1 5 0>)
    )
    Offset Coordinate Space=::Local
    Offset Mode=::NewEnumerator0
    Override Seed=false
    Ring / Disc Distribution Mode:ENiagaraSphereDistributionMode=::NewEnumerator1
    Ring Radius=105
    Rotation Angle=-25
    Rotation Angle Type:ENiagara_AngleInput=::NewEnumerator0
    Rotation Axis=<0 1 0>
    Rotation Coordinate Space=::Simulation
    Rotation Mode=::NewEnumerator1
    Shape Primitive=::NewEnumerator4
    U Position=&Add_Float(
     A=&Multiply_Float(A=$Emitter.Age, B=-0.3)
     B=&ReturnNormalizedExecIndex(Normalized Index Includes One=false)
    )
    UseOwnerScale=true
   ) {module=ShapeLocation001}
  ]
  rendererClasses [&NiagaraMeshRendererProperties]
  renderers [{rendererClass=&NiagaraMeshRendererProperties, rendererIndex=0}]
 }
 InfiniteParticleLifetime004 {
  emitterUpdateScript [
   &EmitterState(
    Inactive Response=::NewEnumerator1
    Life Cycle Mode=::NewEnumerator1
    Loop Behavior=::NewEnumerator1
    Loop Duration Mode=::NewEnumerator1
    UseLoopDelay=false
   )
   &SpawnBurst_Instantaneous(Spawn Count=13, Use Spawn Probability=false, UseLoopCountLimit=false)
  ]
  particleSpawnScript [
   &InitializeParticle(
    Color Mode=::NewEnumerator0
    Mesh Scale Mode=::NewEnumerator1
    Mesh Uniform Scale Max=1.5
    Mesh Uniform Scale Min=1.25
    Position Mode=::NewEnumerator0
    Write Material Random=true
   )
   &InitialMeshOrientation(
    Mesh Orientation Mode=::NewEnumerator1
    Rotation=&RandomRangeVector(Minimum=<0 0 0>, Maximum=<0 0 1>)
    Rotation Coordinate Space=::NewEnumerator2
    Use Rotation Vector=true
   )
  ]
  particleUpdateScript [
   &ParticleState(Kill Particles When Lifetime Has Elapsed=false, Loop Particles Lifetime=true)
   &ShapeLocation(
    MaskSpawnGroup=false
    Non Uniform Scale=<1 0.5 1>
    Offset=&AddVectorToPosition(
     Position=&Lerp_Vector(
      A=<0 0 -110>
      B=<0 0 0>
      Alpha=&ClampFloat(Float=&Multiply_Float(A=$Emitter.Age, B=1.25))
     )
     Vector=&RandomRangeVector(Minimum=<0 -10 0>, Maximum=<0 10 0>)
    )
    Offset Coordinate Space=::Local
    Offset Mode=::NewEnumerator0
    Override Seed=false
    Ring / Disc Distribution Mode:ENiagaraSphereDistributionMode=::NewEnumerator1
    Ring Radius=105
    Rotation Angle=-25
    Rotation Angle Type:ENiagara_AngleInput=::NewEnumerator0
    Rotation Axis=<0 1 0>
    Rotation Coordinate Space=::Simulation
    Rotation Mode=::NewEnumerator1
    Shape Primitive=::NewEnumerator4
    U Position=&Add_Float(
     A=&Multiply_Float(A=$Emitter.Age, B=-0.3)
     B=&ReturnNormalizedExecIndex(Normalized Index Includes One=false)
    )
    UseOwnerScale=true
   ) {module=ShapeLocation001}
  ]
  rendererClasses [&NiagaraMeshRendererProperties]
  renderers [
   {rendererClass=&NiagaraMeshRendererProperties, rendererIndex=0}
   {rendererClass=&NiagaraMeshRendererProperties, rendererIndex=1}
  ]
 }
 Sparkles001 {
  emitterUpdateScript [
   &EmitterState()
   &SpawnRate(
    SpawnRate=&Lerp_Float(
     A=0
     B=50
     Alpha=&ClampFloat(Float=&FloatFromCurve(FloatCurve=DataInterface() {propertyValues {Curve {keys [<at 0 0 RCIM_Constant>,<at 1 1 RCIM_Constant>]}}}, CurveIndex=$Emitter.Age))
    )
    Use Spawn Probability=false
   )
  ]
  particleSpawnScript [
   &InitializeParticle(
    Color Mode=::NewEnumerator0
    Lifetime=&RandomRangeFloat(Minimum=0.2, Maximum=0.4)
    Sprite Size Mode=::NewEnumerator3
    Uniform Sprite Size=&FloatFromCurve(
     FloatCurve=DataInterface() {propertyValues {Curve {keys [<at 0 0.3 RCIM_Constant>,<at 0.4406 0.5 RCIM_Constant>,<at 0.8099 1.5 RCIM_Constant>,<at 1 1.5 RCIM_Constant>]}}}
     CurveIndex=$Particles.Scale
     Scale Curve=&RandomRangeFloat(Minimum=50, Maximum=100)
    )
    UsePositionOffset=false
    Write Material Random=true
   )
   &ShapeLocation(
    Hemisphere Distribution=<1 0.5>
    MaskSpawnGroup=false
    Non Uniform Scale=<1 0.55 0.55>
    Offset Mode=::NewEnumerator0
    Override Seed=false
    Rotation Coordinate Space=::Simulation
    Rotation Mode=::NewEnumerator2
    Sphere Radius=190
    Sphere Surface Distribution=0.75
    UseOwnerScale=false
    Yaw / Pitch / Roll=<0 0.03 0>
   )
  ]
  particleUpdateScript [
   &ParticleState()
   &ScaleSpriteSize(Uniform Curve Sprite Scale=DataInterface() {propertyValues {Curve {keys [{tangentMode=RCTM_User, time=0, value=0}, <at 0.5 1>, {tangentMode=RCTM_User, time=1, value=0}]}}})
   &ScaleColor(
    Scale Alpha=&FloatFromCurve(
     FloatCurve=DataInterface() {propertyValues {Curve {
        defaultValue=1
        keys [
         {interpMode=RCIM_Cubic, tangentMode=RCTM_User, time=0, value=0}
         {interpMode=RCIM_Cubic, tangentMode=RCTM_User, time=0.2269, value=1}
         {arriveTangent=-0.1174, interpMode=RCIM_Cubic, leaveTangent=-0.1174, tangentMode=RCTM_User, time=0.3787, value=1}
         {tangentMode=RCTM_User, time=1, value=0}
        ]
       }}}
     Scale Curve=&Waveform(Global Phase=$Particles.Age, "Waveform [1]"=::NewEnumerator0, "[1] Frequency"=5, "[1] Amplitude Min/Max"=<0 1>, "[1] Bias"=0.3)
    )
    Scale RGB=<1 1 0.25>
    ScaleA=true
    ScaleRGB=true
   )
  ]
  rendererClasses [&NiagaraSpriteRendererProperties]
  renderers [{rendererClass=&NiagaraSpriteRendererProperties}]
 }
}
summary {
 systemName=NS_Emote_FunkyFeather_GroundSwim_v01
 emitters [
  {emitterName=InfiniteParticleLifetime004, rendererClasses [&NiagaraMeshRendererProperties]}
  {emitterName=InfiniteParticleLifetime001, rendererClasses [&NiagaraMeshRendererProperties]}
  {emitterName=InfiniteParticleLifetime, rendererClasses [&NiagaraMeshRendererProperties]}
  {emitterName=Sparkles001, rendererClasses [&NiagaraSpriteRendererProperties]}
 ]
 userVariables []
}
